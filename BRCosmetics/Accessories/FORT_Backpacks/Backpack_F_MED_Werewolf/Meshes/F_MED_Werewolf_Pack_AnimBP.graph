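# Anim Blueprint for the Werewolf backpack cosmetic, copying the parent mesh pose and adding procedural motion.
# Spine twist drives the pack bone's additive tilt.
# Time-based RandomRotatorFunction outputs sway three left/right bone chains, alternating with spring-based, gravity-free AnimDynamics bones.
# Weapon, skydiving and customization events set the material's "Hide Element 01" parameter, based on whether a Werewolf-tagged pickaxe is equipped, resetting it when skydiving begins.
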AnimBlueprint /BRCosmetics/Accessories/FORT_Backpacks/Backpack_F_MED_Werewolf/Meshes/F_MED_Werewolf_Pack_AnimBP
target_skeleton=&/F_MED_Werewolf_Pack_Skeleton
graphs [
 AnimGraph {nodes [
   n1 = "Transform (Modify) Bone - Bone: pack"(
    ComponentPose=Local To Component(LocalPose=Copy Pose From Mesh())
    Alpha=Get backpack_down_rotation() -> (backpack_down_rotation:double)
    Translation="0,0,0"
    Rotation="0,0,35"
   ) {
    rotationMode=BMM_Additive
    rotationSpace=BCS_ParentBoneSpace
    alphaBoolBlend {blendOption=Linear}
    boneToModify {boneName=pack}
    rotation {roll=35}
   }
   n2 = "Transform (Modify) Bone - Bone: dyn_1_1_l"(ComponentPose=@n1, Alpha=1, Translation="0,0,0", Rotation=Get rotator_1_1_l() -> (rotator_1_1_l:Rotator), Scale="1,1,1") {
    rotationMode=BMM_Additive
    rotationSpace=BCS_ParentBoneSpace
    translationMode=BMM_Additive
    translationSpace=BCS_ParentBoneSpace
    alphaBoolBlend {blendOption=Linear}
    boneToModify {boneName=dyn_1_1_l}
   }
   n3 = "Transform (Modify) Bone - Bone: dyn_1_3_l"(ComponentPose=@n2, Alpha=1, Translation="0,0,0", Rotation=Get rotator_1_2_l() -> (rotator_1_2_l:Rotator), Scale="1,1,1") {
    rotationMode=BMM_Additive
    rotationSpace=BCS_ParentBoneSpace
    translationMode=BMM_Additive
    translationSpace=BCS_ParentBoneSpace
    alphaBoolBlend {blendOption=Linear}
    boneToModify {boneName=dyn_1_3_l}
   }
   n4 = "Transform (Modify) Bone - Bone: dyn_2_1_l"(ComponentPose=@n3, Alpha=1, Translation="0,0,0", Rotation=Get rotator_2_1_l() -> (rotator_2_1_l:Rotator), Scale="1,1,1") {
    rotationMode=BMM_Additive
    rotationSpace=BCS_ParentBoneSpace
    translationMode=BMM_Additive
    translationSpace=BCS_ParentBoneSpace
    alphaBoolBlend {blendOption=Linear}
    boneToModify {boneName=dyn_2_1_l}
   }
   n5 = "Transform (Modify) Bone - Bone: dyn_2_3_l"(ComponentPose=@n4, Alpha=1, Translation="0,0,0", Rotation=Get rotator_2_2_l() -> (rotator_2_2_l:Rotator), Scale="1,1,1") {
    rotationMode=BMM_Additive
    rotationSpace=BCS_ParentBoneSpace
    translationMode=BMM_Additive
    translationSpace=BCS_ParentBoneSpace
    alphaBoolBlend {blendOption=Linear}
    boneToModify {boneName=dyn_2_3_l}
   }
   n6 = "Transform (Modify) Bone - Bone: dyn_3_1_l"(ComponentPose=@n5, Alpha=1, Translation="0,0,0", Rotation=Get rotator_3_1_l() -> (rotator_3_1_l:Rotator), Scale="1,1,1") {
    rotationMode=BMM_Additive
    rotationSpace=BCS_ParentBoneSpace
    translationMode=BMM_Additive
    translationSpace=BCS_ParentBoneSpace
    alphaBoolBlend {blendOption=Linear}
    boneToModify {boneName=dyn_3_1_l}
   }
   n7 = "Transform (Modify) Bone - Bone: dyn_3_3_l"(ComponentPose=@n6, Alpha=1, Translation="0,0,0", Rotation=Get rotator_3_2_l() -> (rotator_3_2_l:Rotator), Scale="1,1,1") {
    rotationMode=BMM_Additive
    rotationSpace=BCS_ParentBoneSpace
    translationMode=BMM_Additive
    translationSpace=BCS_ParentBoneSpace
    alphaBoolBlend {blendOption=Linear}
    boneToModify {boneName=dyn_3_3_l}
   }
   n8 = Anim Dynamics - Bone: dyn_1_2_l(ComponentPose=@n7, Alpha=1, ExternalForce="0,0,0") {
    angularDampingOverride=0.98
    angularSpringConstant=10000
    bAngularSpring=true
    bGravityOverrideInSimSpace=true
    bOverrideAngularDamping=true
    gravityScale=0
    simulationSpace=RootRelative
    alphaBoolBlend {blendOption=Linear}
    boundBone {boneName=dyn_1_2_l}
    componentLinearAccScale=<0.5 0.5 0.5>
    componentLinearVelScale=<1 1 1>
    gravityOverride {z=-980}
    physicsBodyDefinitions [{
      sphereCollisionRadius=0
      boundBone {boneName=dyn_1_2_l}
      boxExtents=<22 32 32>
      constraintSetup {angularLimitsMax=<2.5 5 6.5>, angularLimitsMin=<-2.5 -5 -6.5>, angularTarget=<1 0 0>}
      localJointOffset=<0 0 -2>
     }]
    relativeSpaceBone {boneName=root}
   }
   n9 = Anim Dynamics - Bone: dyn_1_4_l(ComponentPose=@n8, Alpha=1, ExternalForce="0,0,0") {
    angularDampingOverride=0.98
    angularSpringConstant=10000
    bAngularSpring=true
    bGravityOverrideInSimSpace=true
    bOverrideAngularDamping=true
    gravityScale=0
    simulationSpace=RootRelative
    alphaBoolBlend {blendOption=Linear}
    boundBone {boneName=dyn_1_4_l}
    componentLinearAccScale=<0.25 0.25 0.25>
    componentLinearVelScale=<1 1 1>
    gravityOverride {z=-980}
    physicsBodyDefinitions [{
      sphereCollisionRadius=0
      boundBone {boneName=dyn_1_4_l}
      boxExtents=<22 32 32>
      constraintSetup {angularLimitsMax=<15 15 15>, angularLimitsMin=<-15 -15 -15>, angularTarget=<1 0 0>}
      localJointOffset=<0 0 -2>
     }]
    relativeSpaceBone {boneName=root}
   }
   n10 = Anim Dynamics - Bone: dyn_2_2_l(ComponentPose=@n9, Alpha=1, ExternalForce="0,0,0") {
    angularDampingOverride=0.98
    angularSpringConstant=10000
    bAngularSpring=true
    bGravityOverrideInSimSpace=true
    bOverrideAngularDamping=true
    gravityScale=0
    simulationSpace=RootRelative
    alphaBoolBlend {blendOption=Linear}
    boundBone {boneName=dyn_2_2_l}
    componentLinearAccScale=<0.5 0.5 0.5>
    componentLinearVelScale=<1 1 1>
    gravityOverride {z=-980}
    physicsBodyDefinitions [{
      sphereCollisionRadius=0
      boundBone {boneName=dyn_2_2_l}
      boxExtents=<22 32 32>
      constraintSetup {angularLimitsMax=<15 15 15>, angularLimitsMin=<-15 -15 -15>, angularTarget=<1 0 0>}
      localJointOffset=<0 0 -2>
     }]
    relativeSpaceBone {boneName=root}
   }
   n11 = Anim Dynamics - Bone: dyn_2_4_l(ComponentPose=@n10, Alpha=1, ExternalForce="0,0,0") {
    angularDampingOverride=0.98
    angularSpringConstant=10000
    bAngularSpring=true
    bGravityOverrideInSimSpace=true
    bOverrideAngularDamping=true
    gravityScale=0
    simulationSpace=RootRelative
    alphaBoolBlend {blendOption=Linear}
    boundBone {boneName=dyn_2_4_l}
    componentLinearAccScale=<0.25 0.25 0.25>
    componentLinearVelScale=<1 1 1>
    gravityOverride {z=-980}
    physicsBodyDefinitions [{
      sphereCollisionRadius=0
      boundBone {boneName=dyn_2_4_l}
      boxExtents=<22 32 32>
      constraintSetup {angularLimitsMax=<15 15 15>, angularLimitsMin=<-15 -15 -15>, angularTarget=<1 0 0>}
      localJointOffset=<0 0 -2>
     }]
    relativeSpaceBone {boneName=root}
   }
   n12 = Anim Dynamics - Bone: dyn_3_2_l(ComponentPose=@n11, Alpha=1, ExternalForce="0,0,0") {
    angularDampingOverride=0.98
    angularSpringConstant=10000
    bAngularSpring=true
    bGravityOverrideInSimSpace=true
    bOverrideAngularDamping=true
    gravityScale=0
    simulationSpace=RootRelative
    alphaBoolBlend {blendOption=Linear}
    boundBone {boneName=dyn_3_2_l}
    componentLinearAccScale=<0.5 0.5 0.5>
    componentLinearVelScale=<1 1 1>
    gravityOverride {z=-980}
    physicsBodyDefinitions [{
      sphereCollisionRadius=0
      boundBone {boneName=dyn_3_2_l}
      boxExtents=<22 32 32>
      constraintSetup {angularLimitsMax=<15 15 15>, angularLimitsMin=<-15 -15 -15>, angularTarget=<1 0 0>}
      localJointOffset=<0 0 -2>
     }]
    relativeSpaceBone {boneName=root}
   }
   n13 = Anim Dynamics - Bone: dyn_3_4_l(ComponentPose=@n12, Alpha=1, ExternalForce="0,0,0") {
    angularDampingOverride=0.98
    angularSpringConstant=10000
    bAngularSpring=true
    bGravityOverrideInSimSpace=true
    bOverrideAngularDamping=true
    gravityScale=0
    simulationSpace=RootRelative
    alphaBoolBlend {blendOption=Linear}
    boundBone {boneName=dyn_3_4_l}
    componentLinearAccScale=<0.25 0.25 0.25>
    componentLinearVelScale=<1 1 1>
    gravityOverride {z=-980}
    physicsBodyDefinitions [{
      sphereCollisionRadius=0
      boundBone {boneName=dyn_3_4_l}
      boxExtents=<22 32 32>
      constraintSetup {angularLimitsMax=<15 15 15>, angularLimitsMin=<-15 -15 -15>, angularTarget=<1 0 0>}
      localJointOffset=<0 0 -2>
     }]
    relativeSpaceBone {boneName=root}
   }
   n14 = "Transform (Modify) Bone - Bone: dyn_1_1_r"(ComponentPose=@n13, Alpha=1, Translation="0,0,0", Rotation=Get rotator_1_1_r() -> (rotator_1_1_r:Rotator), Scale="1,1,1") {
    rotationMode=BMM_Additive
    rotationSpace=BCS_ParentBoneSpace
    translationMode=BMM_Additive
    translationSpace=BCS_ParentBoneSpace
    alphaBoolBlend {blendOption=Linear}
    boneToModify {boneName=dyn_1_1_r}
   }
   n15 = "Transform (Modify) Bone - Bone: dyn_1_3_r"(ComponentPose=@n14, Alpha=1, Translation="0,0,0", Rotation=Get rotator_1_2_r() -> (rotator_1_2_r:Rotator), Scale="1,1,1") {
    rotationMode=BMM_Additive
    rotationSpace=BCS_ParentBoneSpace
    translationMode=BMM_Additive
    translationSpace=BCS_ParentBoneSpace
    alphaBoolBlend {blendOption=Linear}
    boneToModify {boneName=dyn_1_3_r}
   }
   n16 = "Transform (Modify) Bone - Bone: dyn_2_1_r"(ComponentPose=@n15, Alpha=1, Translation="0,0,0", Rotation=Get rotator_2_1_r() -> (rotator_2_1_r:Rotator), Scale="1,1,1") {
    rotationMode=BMM_Additive
    rotationSpace=BCS_ParentBoneSpace
    translationMode=BMM_Additive
    translationSpace=BCS_ParentBoneSpace
    alphaBoolBlend {blendOption=Linear}
    boneToModify {boneName=dyn_2_1_r}
   }
   n17 = "Transform (Modify) Bone - Bone: dyn_2_3_r"(ComponentPose=@n16, Alpha=1, Translation="0,0,0", Rotation=Get rotator_2_2_r() -> (rotator_2_2_r:Rotator), Scale="1,1,1") {
    rotationMode=BMM_Additive
    rotationSpace=BCS_ParentBoneSpace
    translationMode=BMM_Additive
    translationSpace=BCS_ParentBoneSpace
    alphaBoolBlend {blendOption=Linear}
    boneToModify {boneName=dyn_2_3_r}
   }
   n18 = "Transform (Modify) Bone - Bone: dyn_3_1_r"(ComponentPose=@n17, Alpha=1, Translation="0,0,0", Rotation=Get rotator_3_1_r() -> (rotator_3_1_r:Rotator), Scale="1,1,1") {
    rotationMode=BMM_Additive
    rotationSpace=BCS_ParentBoneSpace
    translationMode=BMM_Additive
    translationSpace=BCS_ParentBoneSpace
    alphaBoolBlend {blendOption=Linear}
    boneToModify {boneName=dyn_3_1_r}
   }
   n19 = "Transform (Modify) Bone - Bone: dyn_3_3_r"(ComponentPose=@n18, Alpha=1, Translation="0,0,0", Rotation=Get rotator_3_2_r() -> (rotator_3_2_r:Rotator), Scale="1,1,1") {
    rotationMode=BMM_Additive
    rotationSpace=BCS_ParentBoneSpace
    translationMode=BMM_Additive
    translationSpace=BCS_ParentBoneSpace
    alphaBoolBlend {blendOption=Linear}
    boneToModify {boneName=dyn_3_3_r}
   }
   n20 = Anim Dynamics - Bone: dyn_1_2_r(ComponentPose=@n19, Alpha=1, ExternalForce="0,0,0") {
    angularDampingOverride=0.98
    angularSpringConstant=10000
    bAngularSpring=true
    bGravityOverrideInSimSpace=true
    bOverrideAngularDamping=true
    gravityScale=0
    simulationSpace=RootRelative
    alphaBoolBlend {blendOption=Linear}
    boundBone {boneName=dyn_1_2_r}
    componentLinearAccScale=<0.5 0.5 0.5>
    componentLinearVelScale=<1 1 1>
    gravityOverride {z=-980}
    physicsBodyDefinitions [{
      sphereCollisionRadius=0
      boundBone {boneName=dyn_1_2_r}
      boxExtents=<22 32 32>
      constraintSetup {angularLimitsMax=<2.5 5 6.5>, angularLimitsMin=<-2.5 -5 -6.5>, angularTarget=<1 0 0>}
      localJointOffset=<0 0 -2>
     }]
    relativeSpaceBone {boneName=root}
   }
   n21 = Anim Dynamics - Bone: dyn_1_4_r(ComponentPose=@n20, Alpha=1, ExternalForce="0,0,0") {
    angularDampingOverride=0.98
    angularSpringConstant=10000
    bAngularSpring=true
    bGravityOverrideInSimSpace=true
    bOverrideAngularDamping=true
    gravityScale=0
    simulationSpace=RootRelative
    alphaBoolBlend {blendOption=Linear}
    boundBone {boneName=dyn_1_4_r}
    componentLinearAccScale=<0.25 0.25 0.25>
    componentLinearVelScale=<1 1 1>
    gravityOverride {z=-980}
    physicsBodyDefinitions [{
      sphereCollisionRadius=0
      boundBone {boneName=dyn_1_4_r}
      boxExtents=<22 32 32>
      constraintSetup {angularLimitsMax=<15 15 15>, angularLimitsMin=<-15 -15 -15>, angularTarget=<1 0 0>}
      localJointOffset=<0 0 -2>
     }]
    relativeSpaceBone {boneName=root}
   }
   n22 = Anim Dynamics - Bone: dyn_2_2_r(ComponentPose=@n21, Alpha=1, ExternalForce="0,0,0") {
    angularDampingOverride=0.98
    angularSpringConstant=10000
    bAngularSpring=true
    bGravityOverrideInSimSpace=true
    bOverrideAngularDamping=true
    gravityScale=0
    simulationSpace=RootRelative
    alphaBoolBlend {blendOption=Linear}
    boundBone {boneName=dyn_2_2_r}
    componentLinearAccScale=<0.5 0.5 0.5>
    componentLinearVelScale=<1 1 1>
    gravityOverride {z=-980}
    physicsBodyDefinitions [{
      sphereCollisionRadius=0
      boundBone {boneName=dyn_2_2_r}
      boxExtents=<22 32 32>
      constraintSetup {angularLimitsMax=<15 15 15>, angularLimitsMin=<-15 -15 -15>, angularTarget=<1 0 0>}
      localJointOffset=<0 0 -2>
     }]
    relativeSpaceBone {boneName=root}
   }
   n23 = Anim Dynamics - Bone: dyn_2_4_r(ComponentPose=@n22, Alpha=1, ExternalForce="0,0,0") {
    angularDampingOverride=0.98
    angularSpringConstant=10000
    bAngularSpring=true
    bGravityOverrideInSimSpace=true
    bOverrideAngularDamping=true
    gravityScale=0
    simulationSpace=RootRelative
    alphaBoolBlend {blendOption=Linear}
    boundBone {boneName=dyn_2_4_r}
    componentLinearAccScale=<0.25 0.25 0.25>
    componentLinearVelScale=<1 1 1>
    gravityOverride {z=-980}
    physicsBodyDefinitions [{
      sphereCollisionRadius=0
      boundBone {boneName=dyn_2_4_r}
      boxExtents=<22 32 32>
      constraintSetup {angularLimitsMax=<15 15 15>, angularLimitsMin=<-15 -15 -15>, angularTarget=<1 0 0>}
      localJointOffset=<0 0 -2>
     }]
    relativeSpaceBone {boneName=root}
   }
   n24 = Anim Dynamics - Bone: dyn_3_2_r(ComponentPose=@n23, Alpha=1, ExternalForce="0,0,0") {
    angularDampingOverride=0.98
    angularSpringConstant=10000
    bAngularSpring=true
    bGravityOverrideInSimSpace=true
    bOverrideAngularDamping=true
    gravityScale=0
    simulationSpace=RootRelative
    alphaBoolBlend {blendOption=Linear}
    boundBone {boneName=dyn_3_2_r}
    componentLinearAccScale=<0.5 0.5 0.5>
    componentLinearVelScale=<1 1 1>
    gravityOverride {z=-980}
    physicsBodyDefinitions [{
      sphereCollisionRadius=0
      boundBone {boneName=dyn_3_2_r}
      boxExtents=<22 32 32>
      constraintSetup {angularLimitsMax=<15 15 15>, angularLimitsMin=<-15 -15 -15>, angularTarget=<1 0 0>}
      localJointOffset=<0 0 -2>
     }]
    relativeSpaceBone {boneName=root}
   }
   n25 = Anim Dynamics - Bone: dyn_3_4_r(ComponentPose=@n24, Alpha=1, ExternalForce="0,0,0") {
    angularDampingOverride=0.96
    angularSpringConstant=10000
    bAngularSpring=true
    bGravityOverrideInSimSpace=true
    bOverrideAngularDamping=true
    gravityScale=0
    simulationSpace=RootRelative
    alphaBoolBlend {blendOption=Linear}
    boundBone {boneName=dyn_3_4_r}
    componentLinearAccScale=<0.25 0.25 0.25>
    componentLinearVelScale=<1 1 1>
    gravityOverride {z=-980}
    physicsBodyDefinitions [{
      sphereCollisionRadius=0
      boundBone {boneName=dyn_3_4_r}
      boxExtents=<10 25 25>
      constraintSetup {angularLimitsMax=<15 15 15>, angularLimitsMin=<-15 -15 -15>, angularTarget=<1 0 0>}
      localJointOffset=<0 0 -2>
     }]
    relativeSpaceBone {boneName=root}
   }
   Output Pose(Result=Component To Local(ComponentPose=@n25)) {layerGroup=None}
  ]}
 EventGraph {nodes [
   Event BlueprintUpdateAnimation()
   > Set backpack_down_rotation(backpack_down_rotation:double=GetAlphaFromReferenceTransformDelta(
     SkeletalMeshComponent=Get parent_skeletal_mesh() -> (parent_skeletal_mesh:Skeletal Mesh Component Object Reference)
     BoneName=spine_05
     BaseName=spine_01
     DeltaComponentType=QuaternionTwist
     InputRange=(X=0,Y=45)
     TwistAxis="0,0,1"
    )) -> (Output_Get:double)
   n1 = GetGameTimeInSeconds()
   > n2 = RandomRotatorFunction!(
    execute
    GameTimeInSeconds:double=@n1
    PerlinScale:double=0
    Frequency:Rotator="1,0,0"
    Amplitude:Rotator="4,0,0"
    Offset:Rotator="0, 0, 0"
    TimeOffset:double=0
   ) -> (then, RandomRotator:Rotator)
   > Set rotator_1_1_l(rotator_1_1_l:Rotator=@n2) -> (Output_Get:Rotator)
   > n3 = RandomRotatorFunction(GameTimeInSeconds=@n1, PerlinScale=0, Frequency="1,0,0", Amplitude="3.5,0,0", Offset="0, 0, 0", TimeOffset=0)
   > Set rotator_1_1_r(rotator_1_1_r:Rotator=@n3) -> (Output_Get:Rotator)
   > n4 = RandomRotatorFunction(GameTimeInSeconds=@n1, PerlinScale=0, Frequency="0,1,0", Amplitude="0,5,0", Offset="0, 0, 0", TimeOffset=5)
   > Set rotator_1_2_l(rotator_1_2_l:Rotator=@n4) -> (Output_Get:Rotator)
   > n5 = RandomRotatorFunction(GameTimeInSeconds=@n1, PerlinScale=0, Frequency="0,1.1,0", Amplitude="0,5.1,0", Offset="0, 0, 0", TimeOffset=5.5)
   > Set rotator_1_2_r(rotator_1_2_r:Rotator=@n5) -> (Output_Get:Rotator)
   > n6 = RandomRotatorFunction(GameTimeInSeconds=@n1, PerlinScale=0, Frequency="-1.1,0,0", Amplitude="2.25,0,0", Offset="0, 0, 0", TimeOffset=0)
   > Set rotator_2_1_l(rotator_2_1_l:Rotator=@n6) -> (Output_Get:Rotator)
   > n7 = RandomRotatorFunction(GameTimeInSeconds=@n1, PerlinScale=0, Frequency="-1,0,0", Amplitude="2.35,0,0", Offset="0, 0, 0", TimeOffset=0)
   > Set rotator_2_1_r(rotator_2_1_r:Rotator=@n7) -> (Output_Get:Rotator)
   > n8 = RandomRotatorFunction(GameTimeInSeconds=@n1, PerlinScale=0, Frequency="0,1.2,0", Amplitude="0,2.5,0", Offset="0, 0, 0", TimeOffset=6)
   > Set rotator_2_2_l(rotator_2_2_l:Rotator=@n8) -> (Output_Get:Rotator)
   > n9 = RandomRotatorFunction(GameTimeInSeconds=@n1, PerlinScale=0, Frequency="0,1.35,0", Amplitude="0,2.4,0", Offset="0, 0, 0", TimeOffset=6.5)
   > Set rotator_2_2_r(rotator_2_2_r:Rotator=@n9) -> (Output_Get:Rotator)
   > n10 = RandomRotatorFunction(GameTimeInSeconds=@n1, PerlinScale=0, Frequency="1.4,0,0", Amplitude="2.3,0,0", Offset="0, 0, 0", TimeOffset=0)
   > Set rotator_3_1_l(rotator_3_1_l:Rotator=@n10) -> (Output_Get:Rotator)
   > n11 = RandomRotatorFunction(GameTimeInSeconds=@n1, PerlinScale=0, Frequency="1.5,0,0", Amplitude="2.5,0,0", Offset="0, 0, 0", TimeOffset=0)
   > Set rotator_3_1_r(rotator_3_1_r:Rotator=@n11) -> (Output_Get:Rotator)
   > n12 = RandomRotatorFunction(GameTimeInSeconds=@n1, PerlinScale=0, Frequency="0,1,0", Amplitude="0,2.25,0", Offset="0, 0, 0", TimeOffset=4)
   > Set rotator_3_2_l(rotator_3_2_l:Rotator=@n12) -> (Output_Get:Rotator)
   > n13 = RandomRotatorFunction(GameTimeInSeconds=@n1, PerlinScale=0, Frequency="0,1.1,0", Amplitude="0,2.75,0", Offset="0, 0, 0", TimeOffset=5)
   > Set rotator_3_2_r(rotator_3_2_r:Rotator=@n13) -> (Output_Get:Rotator)
   Event BlueprintInitializeAnimation()
   > Set parent_skeletal_mesh(parent_skeletal_mesh:Skeletal Mesh Component Object Reference=GetOwnerSkeletalMeshComponent()) -> (Output_Get:Skeletal Mesh Component Object Reference)
   n14 = Event OnWeaponChanged()
   n15 = Get Dynamic Material Instance(execute=[@n14.then, @n23.then], SkeletalMeshComponent=GetPartSkeletalMeshComponent())
   > n16 = Sequence()
   n17 = PickaxeItemDefinitionHasAnyMetaTag(Tags=Make Array("[0]"=(TagName="Cosmetics.MetaTags.Werewolf")), bRequiresEquipped=true)
   > n18 = Branch(Condition=@n17)
   > n19 = Branch(Condition=Get bIsSkyDiving())
   SetScalarParameterValue(execute=[@n21.then, @n19.then, @n20.else, @n18.else], self=@n15, ParameterName=Hide Element 01, Value=0)
   SetScalarParameterValue(execute=[@n19.else, @n20.then], self=@n15, ParameterName=Hide Element 01, Value=1)
   n20 = Branch(execute=[@n22.then, @n16.then_1], Condition=@n17)
   n21 = Event OnBeginSkydiving()
   n22 = Event OnEndSkydiving()
   n23 = Event OnCharacterCustomizationLoaded()
  ]}
]
variables [
 backpack_down_rotation {type=number}
 backpack_up_rotation {type=number}
 parent_skeletal_mesh {type=SkeletalMeshComponent}
 rotator_1_1_l {type=Rotator}
 rotator_1_1_r {type=Rotator}
 rotator_1_2_l {type=Rotator}
 rotator_1_2_r {type=Rotator}
 rotator_2_1_l {type=Rotator}
 rotator_2_1_r {type=Rotator}
 rotator_2_2_l {type=Rotator}
 rotator_2_2_r {type=Rotator}
 rotator_3_1_l {type=Rotator}
 rotator_3_1_r {type=Rotator}
 rotator_3_2_l {type=Rotator}
 rotator_3_2_r {type=Rotator}
]
